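; Remote-controlled guided rocket pawn that a Fortnite player flies from the RC launcher.
; Client setup swaps launcher for controller, transforms the missile and starts its effects and audio.
; Tick banks the poseable mesh and wings using a 2D cross product of successive forward vectors, drives banking audio, and updates fuel visuals from a curve-table-length timer.
; Vehicle overlap or water triggers destruction; replicated bKilled runs explosion cosmetics, and unpossession restores the launcher.
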(event EventOnCharacterPartsReinitialized
  ; Turn off character parts visibility so that no T-pose player standing on rocket
  (CallFunction|SetCharacterBodyVisibilityforPossession false))

(event RemoteControlledPawn|EventOnClientSetupRemoteControlPawn
  (bind _missilelauncher (Variables|Default|GetMissileLauncher))
  (bind _delaytogglerc_vis (Variables|Default|GetDelayToggleRC_Vis))
  (bind _rc (Variables|/Game/Athena/Items/Weapons/Abilities/RCRocket/B_RCRocket_Launcher_Athena.B_RCRocket_Launcher_Athena_C|GetRC _missilelauncher))
  (bind _controllingplayerpawn (Variables|RemoteControlledPawn|GetControllingPlayerPawn))
  ;[1 1 1] Set Team color of Blinking Light
  (Utilities|FlowControl|DoOnce false
    (:Completed
      (Networking|CanExecuteCosmeticEvents
        (:True
          (Effects|Components|ParticleSystem|SetColorParameter (Variables|Default|GetBlinkingLight) "TeamColor" (select (Team|OnSameTeam self (Utilities|Array|Get(acopy) (Fortnite|GetLocalFortPlayerControllers true))) "(R=0.000000,G=0.200000,B=0.800000,A=1.000000)" "(R=1.000000,G=0.000000,B=0.000000,A=1.000000)"))
          ;[1 1 1] Delay for Launch Fire FX
          (Utilities|FlowControl|Delay 0.3) ; Delay for FX on Launcher to complete
          ;[1 1 1] Start Animations, and Toggle visibility of Controller / Launcher
          (Utilities|IsValid _controllingplayerpawn
            (:Is%20Valid
              (bind _asb_rcrocket_launcher_athena (Utilities|Casting|CastToB_RCRocket_Launcher_Athena (Variables|/Script/FortniteGame.FortPawn|GetCurrentWeapon _controllingplayerpawn))
                (:then
                  (bind _output_get (Variables|Default|SetMissileLauncher _asb_rcrocket_launcher_athena))
                  (Animation|PlayAnimMontage _controllingplayerpawn (Variables|Default|GetSwapToIdleAnim) 1.0 "None")
                  (Animation|Montage|PlayMontage (Variables|/Script/FortniteGame.FortWeapon|GetWeaponMesh _output_get) "/Game/Animation/Game/MainPlayer/Combat/Guns/GuidedMissile/GuidedMissile_Launcher_WM_M.GuidedMissile_Launcher_WM_M")
                  (Rendering|SetVisibility (Variables|/Script/FortniteGame.FortWeapon|GetWeaponMesh _output_get) false false)
                  (Utilities|FlowControl|Delay _delaytogglerc_vis) ; Delay for anim to play before revealing RC
                  (Utilities|IsValid (Variables|RemoteControlledPawn|GetControllingPlayerPawn)
                    (:Is%20Valid
                      (if (not (Variables|Default|GetbKilled))
                        (Rendering|SetVisibility (Variables|/Game/Athena/Items/Weapons/Abilities/RCRocket/B_RCRocket_Launcher_Athena.B_RCRocket_Launcher_Athena_C|GetRC _output_get) true false)
                        (Animation|Montage|PlayMontage (Variables|/Game/Athena/Items/Weapons/Abilities/RCRocket/B_RCRocket_Launcher_Athena.B_RCRocket_Launcher_Athena_C|GetRC _output_get) (Variables|Default|GetControllerAnim) 1.0 _delaytogglerc_vis)
                        ;[1 1 1] Looping Controller Audio
                        (bind _returnvalue (Audio|SpawnSoundatLocation "/Game/Athena/Sounds/Weapons/Guided_Missile/GuidedMissile_Laptop_Use_Loop_Cue.GuidedMissile_Laptop_Use_Loop_Cue" (Transformation|GetActorLocation (Variables|RemoteControlledPawn|GetControllingPlayerPawn))))
                        (Variables|Default|SetScreenSoundComponent _returnvalue))
                      ;[1 1 1] Missile Transformation
                      (Variables|Default|SetbTransformed true)
                      (Niagara|SetNiagaraVariable(Float) (Variables|Default|GetNS_RocketTrail_RC) "User.ThrusterOffset" 1.0)
                      (Rendering|SetVisibility (Variables|Default|GetLoadMissile) false false)
                      (if (not (Variables|Default|GetbKilled))
                        (Effects|SpawnEmitteratLocation (Variables|Default|GetTransformationPS) (Transformation|GetWorldLocation (Variables|Default|GetLoadMissile)) (Transformation|GetWorldRotation (Variables|Character|GetCapsuleComponent))) ; Missile Transform Particle System
                        (Audio|SpawnSoundAttached (Variables|Default|GetTransformationSound) (Variables|Transformation|GetRootComponent)) ; Transformation Sound
                        (Rendering|SetVisibility (Variables|Default|GetPoseableMissileMesh) true false)
                        (Components|Activation|Activate (Variables|Default|GetBlinkingLight) false)
                        (Utilities|FlowControl|Sequence
                          (:then_0
                            ;[1 1 1] Increase Missile Size and Fade In Audio
                            (bind (_direction _scalerocket) (Timeline|ScaleUpRocket Play)
                              (:Update
                                (Transformation|SetRelativeScale3D (Variables|Default|GetPoseableMissileMesh) _scalerocket))
                              (:Finished
                                (Utilities|FlowControl|Delay 1.0)
                                (if (not (Variables|Default|GetbKilled))
                                  (Audio|Components|Audio|FadeIn (Variables|Default|GetMissileInAirAudio) 1.0 1.0 0.0 "Linear" :!class /Script/Engine.AudioComponent)))))
                          (:then_1
                            ;[1 1 1] Rocket Hologram FX on Controller
                            (Utilities|IsValid (Variables|RemoteControlledPawn|GetControllingPlayerPawn)
                              (:Is%20Valid
                                (Utilities|FlowControl|Delay 0.7) ; Delay so RC finishes Anim
                                (if (Variables|Default|GetbKilled)
                                  _
                                  (elif (Rendering|IsVisible _rc)
                                    (bind _returnvalue_2 (Effects|SpawnEmitterAttached :EmitterTemplate (Variables|Default|GetHologramRocket) :AttachToComponent _rc :AttachPointName "hologramSocket" :Scale "0.800000,0.800000,0.800000" :LocationType "SnapToTarget" :bAutoDestroy true :PoolingMethod "None" :bAutoActivate true))
                                    (bind _output_get_2 (Variables|Default|SetControllerRocketFX _returnvalue_2))
                                    (Rendering|SetRenderCustomDepth _output_get_2 true)))))))))
                    (:Is%20Not%20Valid
                      _)))
                (:CastFailed
                  _)))
            (:Is%20Not%20Valid
              _)))))))

(event EventBeginPlay
  ;[1 1 1] Sets and starts Fuel Timer
  (bind _outxy (DataTable|EvaluateCurveTableRow "/Game/Athena/Balance/DataTables/AthenaGadgetAttributeData.AthenaGadgetAttributeData" "RemoteControlPawn.Defaults.FortRemoteControlPawnSet.FuseLength")
    (:RowFound
      (bind _output_get (Variables|Default|SetMaxFuelAmount _outxy))
      (bind _returnvalue (Utilities|Time|SetTimerbyEvent (delegate Custom|StartFuelTimer) _output_get))
      (Variables|Default|SetFuelTimer _returnvalue))))

(event RemoteControlledPawn|EventOnTriggeredHealthThreshold
  (params (PreviousTriggeredHealthThreshold :type Float%20%28single-precision%29) (CurrentTriggeredHealthThreshold :type Float%20%28single-precision%29))
  ;[1 1 1] Event to handle Health Thresholds (75%, 50%, 30%, 15%, 0%)
  (Utilities|FlowControl|Sequence
    (:then_0
      (if (<= CurrentTriggeredHealthThreshold 0.75)
        (Utilities|FlowControl|DoOnce false
          (:Completed
            ;;[1 1 1] 75% Full Damage
            (Effects|Components|ParticleSystem|SetFloatParameter (Variables|Default|GetNS_RocketTrail_RC) "DamageAmount" 0.25)
            (Development|PrintString "75")))))
    (:then_1
      (if (<= CurrentTriggeredHealthThreshold 0.5)
        (Utilities|FlowControl|DoOnce false
          (:Completed
            ;;[1 1 1] 50% Full Damage
            (Effects|Components|ParticleSystem|SetFloatParameter (Variables|Default|GetNS_RocketTrail_RC) "DamageAmount" 0.5)
            (Development|PrintString "50")))))
    (:then_2
      (if (<= CurrentTriggeredHealthThreshold 0.3)
        (Utilities|FlowControl|DoOnce false
          (:Completed
            ;;[1 1 1] 30% Full Damage
            (Effects|Components|ParticleSystem|SetFloatParameter (Variables|Default|GetNS_RocketTrail_RC) "DamageAmount" 0.8)
            (Development|PrintString "30")))))))

(event Custom|StartFuelTimer)

(event Custom|ResetWeapon
  ;[1 1 1] Stop Anims and swap Controller / Launcher Visibility
  (Audio|Components|Audio|Stop (Variables|Default|GetScreenSoundComponent))
  (Animation|StopAnimMontage (Variables|RemoteControlledPawn|GetControllingPlayerPawn) (Variables|Default|GetSwapToIdleAnim))
  (Utilities|FlowControl|Delay 0.2) ; Avoid Launcher snapping through head
  (Components|DestroyComponent (Variables|SoundIndicator|GetSoundIndicatorComponent))
  (Components|DestroyComponent (Variables|Default|GetControllerRocketFX))
  (Utilities|IsValid (Variables|Default|GetMissileLauncher)
    (:Is%20Valid
      (Rendering|SetVisibility (Variables|/Game/Athena/Items/Weapons/Abilities/RCRocket/B_RCRocket_Launcher_Athena.B_RCRocket_Launcher_Athena_C|GetRC (Variables|Default|GetMissileLauncher)) false false)
      (Rendering|SetVisibility (Variables|/Script/FortniteGame.FortWeapon|GetWeaponMesh (Variables|Default|GetMissileLauncher)) true false)
      (Utilities|FlowControl|Delay 2.0)
      (Components|DestroyComponent (Variables|Default|GetControllerRocketFX)))))

(event EventUnpossessed
  (params (OldController :type Controller%20Object%20Reference))
  ;[1 1 1] Event Unpossessed
  (CallFunction|ResetWeapon)
  (Networking|SwitchHasAuthority
    (:Authority
      (Variables|Default|SetbDetached true))))

(event RemoteControlledPawn|EventOnDoKill
  ;[1 1 1] Handle when the Remote Controlled Pawn is Killed
  (Variables|Default|SetbKilled true)
  (Networking|ForceNetUpdate)
  (Utilities|IsValid (Variables|RemoteControlledPawn|GetControllingPlayerPawn)
    (:Is%20Valid
      (if (FortPawn|IsDead (Variables|RemoteControlledPawn|GetControllingPlayerPawn))
        (CallFunction|ResetWeapon)
        (Rendering|SetVisibility (Variables|Character|GetMesh) false true)))))

(event Custom|OnKilled
  (bind _fadeouttime (Variables|Default|GetFadeOutTime))
  ;[1 1 1] Handle when the Remote Controlled Pawn is Killed
  (Networking|CanExecuteCosmeticEvents
    (:True
      (Niagara|SpawnSystematLocation (Variables|Default|GetNS_Explosion) (Transformation|GetActorLocation) (Transformation|GetSocketRotation (Variables|Default|GetPoseableMissileMesh) "ExplosionRotation"))
      (Audio|Components|Audio|FadeOut (Variables|Default|GetMissileInAirAudio) (Variables|Default|GetFadeOutTime) 0.0 "Linear" :!class /Script/Engine.AudioComponent)
      (Audio|PlaySoundatLocation (Variables|Default|GetExplosionSound) (Transformation|GetActorLocation))
      (Rendering|SetVisibility (Variables|Character|GetMesh) false true)
      (Rendering|SetActorHiddenInGame true)
      (Utilities|FlowControl|Sequence
        (:then_0
          (bind _missile_lowfuelwarning (Variables|Default|GetMissile_LowFuelWarning)
            (:then
              (Audio|Components|Audio|FadeOut _missile_lowfuelwarning _fadeouttime 0.0 "Linear" :!class /Script/Engine.AudioComponent))
            (:else)))
        (:then_1
          (bind _screensoundcomponent (Variables|Default|GetScreenSoundComponent)
            (:then
              (Audio|Components|Audio|FadeOut _screensoundcomponent _fadeouttime 0.0 "Linear" :!class /Script/Engine.AudioComponent))
            (:else)))))))

(event Custom|Handle%20Movement%20Audio
  (bind _missileinairbankingaudio (Variables|Default|GetMissileInAirBankingAudio))
  ;[1 1 1] Audio based off missile movement/rotation
  (if (Variables|Default|GetbTransformed)
    (if (not (Audio|Components|Audio|IsPlaying _missileinairbankingaudio))
      (Utilities|FlowControl|DoOnce false
        (:Completed
          (Audio|Components|Audio|Play _missileinairbankingaudio 0.0))))
    (Variables|Audio|SetTurnAudioModulator (Math|Float|Absolute(Float) (Math|Trig|Acos(Degrees) (Math|Vector|DotProduct (Math|Vector|Normalize (Variables|Default|GetCurrentVector)) (Math|Vector|Normalize (Variables|Default|GetPreviousVector))))))
    (bind _output_get (Variables|Audio|SetBankingAudioVolume (Math|Interpolation|FInterpTo (Variables|Audio|GetBankingAudioVolume) (Math|Float|MapRangeClamped (Variables|Audio|GetTurnAudioModulator) 0.2 (Variables|Audio|GetTurnAudioModulatorMax) 0.0 0.9) (Utilities|Time|GetWorldDeltaSeconds) 1.0)))
    (Audio|Parameter|SetFloatParameter (Variables|Default|GetMissileInAirBankingAudio) "BankingSound" _output_get :!class /Script/AudioExtensions.AudioParameterControllerInterface)))

(event EventTick
  (params (DeltaSeconds :type Float%20%28single-precision%29))
  (bind _previousvector (Variables|Default|GetPreviousVector))
  (bind _x (.x _previousvector))
  (bind _y (.y _previousvector))
  (bind _currentvector (Variables|Default|GetCurrentVector))
  (bind _x_2 (.x _currentvector))
  (bind _y_2 (.y _currentvector))
  (bind _returnvalue (* _x_2 -1.0))
  (bind _returnvalue_2 (* _y_2 _x))
  (bind _returnvalue_3 (* _returnvalue _y))
  (bind _returnvalue_4 (+ _returnvalue_3 _returnvalue_2))
  (bind _bottomwing (Variables|Default|GetBottomWing))
  (bind _poseablemissilemesh (Variables|Default|GetPoseableMissileMesh))
  (bind _leftwing (Variables|Default|GetLeftWing))
  (bind _poseablemissilemesh_2 (Variables|Default|GetPoseableMissileMesh))
  (bind _rightwing (Variables|Default|GetRightWing))
  (bind _poseablemissilemesh_3 (Variables|Default|GetPoseableMissileMesh))
  (bind _poseablemissilemesh_4 (Variables|Default|GetPoseableMissileMesh))
  (bind _topwing (Variables|Default|GetTopWing))
  ;[1 1 1] Rotate Missile
  (Networking|CanExecuteCosmeticEvents
    (:True
      (if (not (Variables|Default|GetbKilled))
        (Variables|Default|SetTick%20Delta DeltaSeconds)
        (Variables|Default|SetCurrentVector (Transformation|GetActorForwardVector))
        (CallFunction|FuelLevelVisuals)
        (if (Math|Vector|NotEqual(Vector) (Variables|Default|GetPreviousVector) "0, 0, 0" 1e-05)
          (CallFunction|HandleMovementAudio)
          (if (< (Math|Float|Absolute(Float) _returnvalue_4) 0.004)
            (CallFunction|LevelOutMissile)
            (elif (> _returnvalue_4 0.0)
              (bind _output_get (Variables|Default|SetLevelOutRotation (.pitch (Variables|/Script/Engine.SceneComponent|GetRelativeRotation (Variables|Default|GetPoseableMissileMesh)))))
              (if (< _output_get 35.0)
                (Variables|Default|SetLevelOutRotation (Math|Float|Min(Float) (+ (Variables|Default|GetLevelOutRotation) (* (Variables|Default|GetTurn%20Rate) (Variables|Default|GetTick%20Delta))) 35.0))
                (Transformation|SetRelativeRotation (Variables|Default|GetPoseableMissileMesh) (Math|Rotator|MakeRotator 0.0 (Variables|Default|GetLevelOutRotation)) false false)
                ;[1 1 1] Rotate Wings
                (Components|PoseableMesh|SetBoneRotationbyName _poseablemissilemesh _leftwing (Math|Rotator|MakeRotator (* (Variables|Default|GetLevelOutRotation) -0.6)) "ComponentSpace")
                (Components|PoseableMesh|SetBoneRotationbyName _poseablemissilemesh_2 _rightwing (Math|Rotator|MakeRotator (* (Variables|Default|GetLevelOutRotation) -0.6) 0.0 1.8e+02) "ComponentSpace")
                (Components|PoseableMesh|SetBoneRotationbyName _poseablemissilemesh_3 _topwing (Math|Rotator|MakeRotator 9e+01 9e+01 (* (Variables|Default|GetLevelOutRotation) -0.4)) "ComponentSpace")
                (Components|PoseableMesh|SetBoneRotationbyName _poseablemissilemesh_4 _bottomwing (Math|Rotator|MakeRotator 9e+01 -9e+01 (* (Variables|Default|GetLevelOutRotation) -0.4)) "ComponentSpace")
                (CallFunction|SetPreviousVector)
                (else
                  (CallFunction|SetPreviousVector)))
              (else
                (bind _output_get_2 (Variables|Default|SetLevelOutRotation (.pitch (Variables|/Script/Engine.SceneComponent|GetRelativeRotation (Variables|Default|GetPoseableMissileMesh)))))
                (if (> _output_get_2 -35.0)
                  (Variables|Default|SetLevelOutRotation (Math|Float|Max(Float) (- (Variables|Default|GetLevelOutRotation) (* (Variables|Default|GetTurn%20Rate) (Variables|Default|GetTick%20Delta))) -35.0))
                  (Transformation|SetRelativeRotation (Variables|Default|GetPoseableMissileMesh) (Math|Rotator|MakeRotator 0.0 (Variables|Default|GetLevelOutRotation)) false false)
                  (Components|PoseableMesh|SetBoneRotationbyName _poseablemissilemesh _leftwing (Math|Rotator|MakeRotator (* (Variables|Default|GetLevelOutRotation) -0.6)) "ComponentSpace")
                  (Components|PoseableMesh|SetBoneRotationbyName _poseablemissilemesh_2 _rightwing (Math|Rotator|MakeRotator (* (Variables|Default|GetLevelOutRotation) -0.6) 0.0 -1.8e+02) "ComponentSpace")
                  (Components|PoseableMesh|SetBoneRotationbyName _poseablemissilemesh_3 _topwing (Math|Rotator|MakeRotator 9e+01 9e+01 (* (Variables|Default|GetLevelOutRotation) -0.4)) "ComponentSpace")
                  (Components|PoseableMesh|SetBoneRotationbyName _poseablemissilemesh_4 _bottomwing (Math|Rotator|MakeRotator 9e+01 -9e+01 (* (Variables|Default|GetLevelOutRotation) -0.4)) "ComponentSpace")
                  (CallFunction|SetPreviousVector)
                  (else
                    (CallFunction|SetPreviousVector))))))
          (else
            (CallFunction|SetPreviousVector)))))))

(event Custom|SetPreviousVector
  ;[1 1 1] Rotate Missile
  ;;[1 1 1] Set Previous Vector
  (Variables|Default|SetPreviousVector (Transformation|GetActorForwardVector)))

(event Custom|LevelOutMissile
  ;[1 1 1] Level Out Missile
  (bind _output_get (Variables|Default|SetLevelOutRotation (.pitch (Variables|/Script/Engine.SceneComponent|GetRelativeRotation (Variables|Default|GetPoseableMissileMesh)))))
  (if (Math|Float|NearlyEqual(Float) _output_get)
    (CallFunction|SetPreviousVector)
    (elif (> (Variables|Default|GetLevelOutRotation) 0.0)
      (Variables|Default|SetLevelOutRotation (Math|Float|Max(Float) (- (Variables|Default|GetLevelOutRotation) (* (Variables|Default|GetTurn%20Rate) (Variables|Default|GetTick%20Delta)))))
      (Transformation|SetRelativeRotation (Variables|Default|GetPoseableMissileMesh) (Math|Rotator|MakeRotator 0.0 (Variables|Default|GetLevelOutRotation)) false false)
      ;[1 1 1] Level Out Wings
      (Components|PoseableMesh|SetBoneRotationbyName (Variables|Default|GetPoseableMissileMesh) (Variables|Default|GetLeftWing) (Math|Rotator|MakeRotator (* (Variables|Default|GetLevelOutRotation) -0.6)) "ComponentSpace")
      (Components|PoseableMesh|SetBoneRotationbyName (Variables|Default|GetPoseableMissileMesh) (Variables|Default|GetRightWing) (Math|Rotator|MakeRotator (* (Variables|Default|GetLevelOutRotation) -0.6) 0.0 -1.8e+02) "ComponentSpace")
      (Components|PoseableMesh|SetBoneRotationbyName (Variables|Default|GetPoseableMissileMesh) (Variables|Default|GetTopWing) (Math|Rotator|MakeRotator 9e+01 9e+01 (* (Variables|Default|GetLevelOutRotation) -0.4)) "ComponentSpace")
      (CallFunction|SetPreviousVector)
      (else
        (Variables|Default|SetLevelOutRotation (Math|Float|Min(Float) (+ (Variables|Default|GetLevelOutRotation) (* (Variables|Default|GetTurn%20Rate) (Variables|Default|GetTick%20Delta)))))))))

(event OnComponentBeginOverlap(HitBox)
  (params (OverlappedComponent :type Primitive%20Component%20Object%20Reference) (OtherActor :type Actor%20Object%20Reference) (OtherComp :type Primitive%20Component%20Object%20Reference) (OtherBodyIndex :type Integer) (bFromSweep :type Boolean) (SweepResult :type Hit%20Result%20Structure%20%28by%20ref%29))
  ;[1 1 1] Trigger the RC Pawn for explosion when colliding with a Vehicle
  (Utilities|Casting|CastToFortVehicleInterface OtherActor)
  (Networking|SwitchHasAuthority
    (:Authority
      (if (not (RemoteControlledPawn|IsCurrentlyDelayingLaunch))
        (RemoteControlledPawn|TriggerDestroyRemoteControlPawn)))))

(event WaterBody|EventOnBeginWaterInteraction
  (params (WaterSurfaceLocation :type Vector :default "0, 0, 0") (WaterDepth :type Float%20%28single-precision%29))
  ;[1 1 1] OnHItWater - probably needs a code solution over what's below
  (|CallParentFunction WaterSurfaceLocation WaterDepth)
  (Niagara|SpawnSystematLocation (Variables|Default|GetNiagaraWaterImpactFX) (Transformation|GetActorLocation))
  (RemoteControlledPawn|TriggerDestroyRemoteControlPawn))

(fn FuelLevelVisuals
  (bind _fueltimer (Variables|Default|GetFuelTimer))
  (bind _returnvalue (Utilities|Time|GetTimerRemainingTimebyHandle _fueltimer))
  (bind _maxfuelamount (Variables|Default|GetMaxFuelAmount))
  (bind _returnvalue_2 (* _maxfuelamount 0.5))
  (bind _returnvalue_3 (Math|Float|MapRangeClamped _returnvalue _returnvalue_2 0.0 1.0))
  (Effects|Components|ParticleSystem|SetFloatParameter (Variables|Default|GetNS_RocketTrail_RC) "FuelAmount" _returnvalue_3)
  (if (Math|Boolean|ANDBoolean (<= _returnvalue_3 0.3) (Pawn|IsLocallyControlled) (not (Variables|Default|GetbHasPlayedFuelWarningSound)))
    (Variables|Default|SetbHasPlayedFuelWarningSound true)
    (if (not (Variables|Default|GetbKilled))
      (bind _returnvalue_4 (Audio|SpawnSoundAttached :Sound (Variables|Default|GetMissileLowFuelSound) :AttachToComponent (Variables|Default|GetNS_RocketTrail_RC) :AttachPointName "None" :LocationType "KeepRelativeOffset" :bStopWhenAttachedToDestroyed true :VolumeMultiplier 1.0 :PitchMultiplier 1.0 :StartTime 0.0 :bAutoDestroy true))
      (Variables|Default|SetMissile_LowFuelWarning _returnvalue_4))))

(fn OnRep_bDetached
  (Variables|/Script/Engine.CharacterMovementComponent|SetbRunPhysicsWithNoController true (Variables|Character|GetCharacterMovement)))

(fn OnRep_bKilled
  (if (Variables|Default|GetbKilled)
    (CallFunction|OnKilled)))

(fn ConstructionScript
  (Niagara|SetNiagaraVariable(Float) (Variables|Default|GetNS_RocketTrail_RC) "User.ThrusterOffset" 0.0))
