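; Blueprint for the Athena RC Rocket launcher weapon, handling its cosmetic muzzle effect and rocket model visibility.
; When the weapon becomes visible, a retriggerable two-second delay runs before the RC component is hidden.
; A muzzle flash event activates a rear-muzzle Niagara effect only where cosmetic events can execute.
; It overrides the projectile-weapon check to always return true and defers construction to its parent.
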(event EventOnWeaponVisibilityChanged
  (params (bVisible :type Boolean) (bSetForLocalControllerOnly :type Boolean))
  (if bVisible
    (Utilities|FlowControl|RetriggerableDelay 2.0)
    (Rendering|SetVisibility (Variables|Default|GetRC) false false)))

(event Custom|SendMuzzleFlashFX
  (Networking|CanExecuteCosmeticEvents
    (:True
      (Components|Activation|Activate (Variables|Default|GetNS_RocketLauncher_RearMuzzle_Athena) false))))

(fn IsCachedIsProjectileWeapon
  (return true))

(fn ConstructionScript
  (|CallParentFunction))
